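# Physics asset for the necklace on the M_MED_Alfredo male player cosmetic body, set up for dynamic necklace motion against the upper torso.
# A three-link chain hangs from a tiny master body through tightly limited constraints: a fanned three-capsule top link, a capsule middle and a box end.
# Spine and clavicle capsules provide body collision shapes.
# Only the right clavicle is constrained to spine_05, with free swing and twist.
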PhysicsAsset /BRCosmetics/Characters/Player/Male/Medium/Bodies/M_MED_Alfredo/Meshes/M_MED_Alfredo_Necklace_Physics
bodies {
 clavicle_l [{
   shapeType=Capsule
   length=14.2957
   radius=8.1022
   shapeName=clavicle_l_capsule
   center=<6.4236 1.6902 0>
   rotation=<rot -90 179.9851 -179.9853>
  }]
 clavicle_r [{
   shapeType=Capsule
   length=14.2957
   radius=8.1022
   shapeName=clavicle_l_capsule
   center=<-6.4236 -1.6902 0>
   rotation=<rot -90 0.0099 179.9899>
  }]
 dyn_necklace_01 [
  {shapeType=Capsule, length=3.7165, radius=1.5902, shapeName=dyn_necklace_01_capsule, center=<3.2022 -1.2589 0>, rotation=<rot -90 0 -0.0002>}
  {
   shapeType=Capsule
   length=3.7165
   radius=1.5902
   shapeName=dyn_necklace_01_capsule
   center=<2.3978 -1.1563 -3.5139>
   rotation=<rot -65 -0.0004 0>
  }
  {
   shapeType=Capsule
   length=3.7165
   radius=1.5902
   shapeName=dyn_necklace_01_capsule
   center=<2.3978 -1.1563 3.5139>
   rotation=<rot 65 -0.0004 0>
  }
 ]
 dyn_necklace_02 [{shapeType=Capsule, length=3.2372, radius=1.5902, shapeName=dyn_necklace_02_capsule, center=<2.423 -1.2451 0>, rotation=<rot -90 0 -0.0003>}]
 dyn_necklace_03 [{
   shapeType=Box
   extentX=5.4581
   extentY=1.4269
   extentZ=3.4742
   shapeName=dyn_necklace_03_box
   center=<2.7197 -0.1518 0>
   rotation=<rot 0 4.9999 0>
  }]
 dyn_necklace_mstr [{shapeType=Capsule, length=0, radius=0.5, shapeName=dyn_necklace_mstr_capsule}]
 spine_04 [{shapeType=Capsule, length=10.7403, radius=11.9195, shapeName=spine_04_capsule, center=<1.4159 0.1905 0>}]
 spine_05 [{shapeType=Capsule, length=10.9078, radius=12.1523, shapeName=spine_05_capsule, center=<-3.514 -0.4297 0>}]
}
constraints [
 {bone1Name=dyn_necklace_03, bone2Name=dyn_necklace_02, swing1LimitDegrees=15, swing2LimitDegrees=5, twistLimitDegrees=2}
 {bone1Name=dyn_necklace_02, bone2Name=dyn_necklace_01, swing1LimitDegrees=15, swing2LimitDegrees=5, twistLimitDegrees=2}
 {bone1Name=dyn_necklace_01, bone2Name=dyn_necklace_mstr, swing1LimitDegrees=20, swing2LimitDegrees=2, twistLimitDegrees=2}
 {
  bone1Name=clavicle_r
  bone2Name=spine_05
  swing1LimitDegrees=45
  swing1Motion=Free
  swing2LimitDegrees=45
  swing2Motion=Free
  twistLimitDegrees=45
  twistMotion=Free
 }
]
